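# Translucent, unlit, two-sided Niagara ribbon and mesh-particle material for the Lively Domino pickaxe trail.
# Panning noise distorts a U gradient that samples a color curve atlas, tinted by particle color.
# Opacity combines noisy leading-edge, top-edge and border masks, boosted by cell noise with a Floor-stepped accent.
# A rotated plasma dissolve driven by dynamic parameters erodes it, with depth fade skipped on forward and mobile paths.
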Material /BRCosmetics/Weapons/FORT_Melee/Pickaxe_LivelyDomino/FX/M_Pickaxe_LivelyDomino_trail
graph {
 properties {
  bAutomaticallySetUsageInEditor=false
  bUsedWithNiagaraMeshParticles=true
  bUsedWithNiagaraRibbons=true
  blendMode=BLEND_Translucent
  shadingModel=MSM_Unlit
  twoSided=true
  usedShadingModels=Unlit
 }
 nodes [
  n1 = Multiply(
   A=TextureCoordinate()
   B=AppendVector(
    A=ScalarParameter() {defaultValue=0.3, group=Base, parameterName=BaseTextureU}
    B=ScalarParameter() {defaultValue=1.5, group=Base, parameterName=BaseTextureV}
   )
  )
  n2 = DynamicParameter() {defaultValue=<rgba 0 1 1 1>, paramNames [Dissolve, DynamicU_Offset, DynamicV_Offset, Param4]}
  n3 = Add(A=@n1, B=AppendVector(A=@n2.DynamicU_Offset, B=@n2.DynamicU_Offset))
  n4 = ScalarParameter() {parameterName=UseDynamicOffset}
  n5 = ScalarParameter() {defaultValue=-1, group=Base, parameterName=BaseTextureSpeedX}
  n6 = ScalarParameter() {group=Base, parameterName=BaseTextureSpeedY}
  n7 = TextureSampleParameter2D(UVs=Panner(Coordinate=LinearInterpolate(A=@n1, B=@n3, Alpha=@n4), Speed=AppendVector(A=@n5, B=@n6)) {speedX=-1}) {group=Base, parameterName=BaseTexture, texture=&T_NoisePack_01}
  n8 = Add(A=ComponentMask(None=TextureCoordinate()) {r=true}, B=Multiply(A=@n7.R) {constB=0.1})
  n9 = CurveAtlasRowParameter(CurveTime=@n8) {
   atlas=&/CAtlas_Pickaxe_LivelyDomino_ColorCurve
   curve=&/CV_Pickaxe_LivelyDomino_ColorCurve_Trail
   defaultValue=4
   group=Emissive
   parameterName=ColorCurve
   sortPriority=0
  }
  n10 = ScalarParameter() {defaultValue=1, parameterName=UseColorCurve}
  n11 = ParticleColor()
  MaterialFunctionCall(Emissive=Multiply(A=LinearInterpolate(A=@n8, B=@n9, Alpha=@n10), B=@n11.RGB)) {materialFunction=&MF_FX_EmissiveAdjustments}
  n12 = ChannelMaskParameter(None=@n7.R) {group=Soft Leading Edge, parameterName=Soft Leading Edge Texture Mask, sortPriority=5}
  n13 = TextureCoordinate()
  n14 = Add(A=Multiply(A=Subtract(A=@n12.R) {constB=0.5}) {constB=0.5}, B=OneMinus(None=ComponentMask(None=@n13) {r=true}))
  n15 = SphereMask(A=TextureCoordinate() {uTiling=0.5}, B=Constant2Vector() {g=1, r=0.85}, Radius=Constant() {r=1}, Hardness=Constant() {r=0.5})
  n16 = LinearInterpolate(Alpha=Power(Base=OneMinus(None=@n15)) {constExponent=5}) {constB=0.5}
  n17 = ScalarParameter() {defaultValue=1, parameterName=AddTopEdge}
  n18 = ComponentMask(None=TextureCoordinate()) {g=true}
  n19 = Saturate(None=Multiply(A=@n18) {constB=5})
  n20 = Saturate(None=Multiply(A=OneMinus(None=@n18)) {constB=5})
  n21 = ComponentMask(None=TextureCoordinate()) {r=true}
  n22 = Saturate(None=Multiply(A=@n21))
  n23 = Saturate(None=Multiply(A=OneMinus(None=@n21)))
  n24 = Multiply(
   A=Add(A=Power(Base=@n14) {constExponent=3}, B=LinearInterpolate(B=@n16, Alpha=@n17))
   B=Multiply(A=Multiply(A=@n19, B=@n20), B=Multiply(A=@n22, B=@n23))
  )
  n25 = Multiply(
   A=TextureCoordinate()
   B=AppendVector(
    A=ScalarParameter() {defaultValue=1.2, group=MaskLerp, parameterName=MaskLerpTextureU}
    B=ScalarParameter() {defaultValue=1, group=MaskLerp, parameterName=MaskLerpTextureV}
   )
  )
  n26 = DynamicParameter() {defaultValue=<rgba 0 1 1 1>, paramNames [Dissolve, DynamicU_Offset, DynamicV_Offset, Param4]}
  n27 = Add(B=AppendVector(A=@n26.DynamicU_Offset, B=@n26.DynamicU_Offset))
  n28 = ScalarParameter() {parameterName=UseDynamicOffset}
  n29 = ScalarParameter() {defaultValue=0.75, group=MaskLerp, parameterName=MaskLerpSpeedX}
  n30 = ScalarParameter() {defaultValue=0.2, group=MaskLerp, parameterName=MaskLerpSpeedY}
  n31 = TextureSampleParameter2D(UVs=Panner(Coordinate=LinearInterpolate(A=@n25, B=@n27, Alpha=@n28), Speed=AppendVector(A=@n29, B=@n30)) {speedX=0.75, speedY=0.2}) {
   group=MaskLerp
   parameterName=MaskLerpTexture
   samplerType=SAMPLERTYPE_Grayscale
   texture=/BRCosmetics/Weapons/FORT_Melee/Pickaxe_Cube_Ninja_Male/FX/T_CellNoise
  }
  n32 = ScalarParameter() {defaultValue=15, group=MaskLerp, parameterName=MaskLerpTexture_Int}
  n33 = Add(A=@n24, B=Multiply(A=Multiply(A=@n24) {constB=2}, B=Multiply(A=@n31.R, B=@n32) {constB=20}))
  n34 = Add(
   A=Multiply(A=@n33) {constB=0.5}
   B=Multiply(A=Floor(None=@n33), B=ScalarParameter() {defaultValue=0.2, parameterName=Floor_Intensity}) {constB=0.2}
  )
  n35 = NamedRerouteUsage()
  n36 = Multiply(A=@n11.A, B=Saturate(None=Multiply(A=@n34, B=@n35)))
  n37 = DepthFade(Opacity=@n36, FadeDistance=ScalarParameter() {defaultValue=5, parameterName=DepthFade})
  Multiply(
   A=Power(
    Base=ShadingPathSwitch(Default=@n37, Deferred=@n37, Forward=@n36, Mobile=@n36)
    Exp=ScalarParameter() {defaultValue=1, parameterName=All_MaskPower}
   )
   B=ScalarParameter() {defaultValue=1, parameterName=All_Alpha}
  )
  n38 = Multiply(
   A=TextureCoordinate() {vTiling=0.5}
   B=AppendVector(
    A=ScalarParameter() {defaultValue=1, group=DissolveMask, parameterName=Dissolve A Texture U}
    B=ScalarParameter() {defaultValue=1, group=DissolveMask, parameterName=Dissolve A Texture V}
   )
  )
  n39 = DynamicParameter() {defaultValue=<rgba 0 1 1 1>, paramNames [Dissolve, DynamicU_Offset, DynamicV_Offset, Param4]}
  n40 = Add(B=AppendVector(A=@n39.DynamicU_Offset, B=@n39.DynamicU_Offset))
  n41 = ScalarParameter() {parameterName=UseDynamicOffset}
  n42 = ScalarParameter() {defaultValue=-0.2, group=DissolveMask, parameterName=Dissolve A Speed X}
  n43 = ScalarParameter() {defaultValue=0.1, group=DissolveMask, parameterName=Dissolve A Speed Y}
  n44 = TextureSampleParameter2D(UVs=Panner(Coordinate=LinearInterpolate(A=@n38, B=@n40, Alpha=@n41), Speed=AppendVector(A=@n42, B=@n43)) {speedX=-0.2, speedY=0.1}) {
   group=DissolveMask
   parameterName=Dissolve A Texture
   samplerType=SAMPLERTYPE_LinearGrayscale
   sortPriority=0
   texture=/Game/Effects/Fort_Effects/Emotes/UltraLight/Texture/T_Plasma_C_Greyscale
  }
  n45 = ChannelMaskParameter(None=@n44.RGB) {group=DissolveMask, parameterName=Dissolve A Channel, sortPriority=1}
  n46 = ScalarParameter() {defaultValue=1, group=DissolveMask, parameterName=Dissolve_power}
  n47 = MaterialFunctionCall(
   UVs=TextureCoordinate()
   "Rotation Angle (0-1)"=ScalarParameter() {defaultValue=-0.28, group=DissolveMask, parameterName=DissolveMaskRotator}
  ) {materialFunction=&CustomRotator}
  n48 = AppendVector(
   A=ScalarParameter() {group=DissolveMask, parameterName=DissloveMaskAdd_X}
   B=ScalarParameter() {defaultValue=-1, group=DissolveMask, parameterName=DissolveMaskAdd_Y}
  )
  n49 = Add(
   A=LinearInterpolate(
    A=DynamicParameter() {defaultValue=<rgba 0 1 1 1>, paramNames [Dissolve, DynamicU_Offset, DynamicV_Offset, Param4]}.Dissolve
    Alpha=Power(Base=@n45.R, Exp=@n46) {constExponent=1}
   ) {constA=0.3}
   B=ComponentMask(None=Add(A=@n47.Rotated Values, B=@n48)) {g=true}
  )
  n50 = ScalarParameter() {defaultValue=10, parameterName=Dissolve A Texture_Int}
  NamedRerouteDeclaration(None=Saturate(None=Multiply(A=@n49, B=@n50) {constB=10})) {name=DissloveMask, nodeColor=<rgba 0.0312 0 1 1>}
 ]
}
