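# Physics asset for the LivelyDomino pickaxe, setting up two dangling dice chains, one long and one short, hung off the two-handed handle.
# Each chain has thin capsules ending in a fatter one, with swing limits tightening toward the tip and almost no twist.
# Free joints link the chain bases to the handle.
# It also includes capsules for pelvis, lower spine, thighs and right calf, plus an unconstrained, oversized dyn_root_simspace body.
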PhysicsAsset /BRCosmetics/Weapons/FORT_Melee/Pickaxe_LivelyDomino/Meshes/LivelyDomino_Axe_Physics
bodies {
 Melee_TwoHanded_Handle [{
   shapeType=Capsule
   length=51.9935
   radius=9.8624
   shapeName=Melee_TwoHanded_Handle_capsule
   center=<0.0008 7.1198 4.8516>
   rotation=<rot -0.0087 0 -2.0309>
  }]
 base_dice [{shapeType=Capsule, length=2, radius=1, shapeName=base_dice_capsule}]
 base_longDice_00 [{shapeType=Capsule, length=2, radius=1, shapeName=base_longDice_00_capsule, center=<2 0 0>, rotation=<rot 90 0 0>}]
 base_shortDice_00 [{shapeType=Capsule, length=2, radius=1, shapeName=base_shortDice_00_capsule, center=<2 0 0>, rotation=<rot 90 0 0>}]
 calf_r [{
   shapeType=Capsule
   length=34.5929
   radius=9.6166
   shapeName=thigh_r_capsule
   center=<22.1907 -1.2551 -2.52>
   rotation=<rot 90 83.2785 88.1144>
  }]
 dyn_longDice_00 [{shapeType=Capsule, length=3.6351, radius=0.5664, shapeName=dyn_longDice_00_capsule, center=<1.6556 0 -0.3147>, rotation=<rot 90 0 0>}]
 dyn_longDice_01 [{shapeType=Capsule, length=6.612, radius=0.5649, shapeName=dyn_longDice_01_capsule, center=<2.163 0 0>, rotation=<rot 90 0 0>}]
 dyn_longDice_02 [{shapeType=Capsule, length=6.427, radius=0.5, shapeName=dyn_longDice_02_capsule, center=<2.5362 0 0>, rotation=<rot 90 0 0>}]
 dyn_longDice_03 [{shapeType=Capsule, length=5.5385, radius=0.5, shapeName=dyn_longDice_03_capsule, center=<2.2199 0 0>, rotation=<rot 90 0 0>}]
 dyn_longDice_04 [{shapeType=Capsule, length=0, radius=5.2516, shapeName=dyn_longDice_03_capsule, center=<4.1866 0.11 0.8829>, rotation=<rot 90 0 0>}]
 dyn_root_simspace [{shapeType=Capsule, length=0.505, radius=300, shapeName=dyn_root_simspace_capsule, center=<-339.3039 0.0001 -7.5951>}]
 dyn_shortDice_00 [{
   shapeType=Capsule
   length=4.7226
   radius=0.5
   shapeName=dyn_shortDice_00_capsule
   center=<2.0198 0 0.266>
   rotation=<rot 88.7151 179.9962 179.996>
  }]
 dyn_shortDice_01 [{shapeType=Capsule, length=5.9052, radius=0.5, shapeName=dyn_shortDice_01_capsule, center=<2.0892 0 0>, rotation=<rot 90 0 0>}]
 dyn_shortDice_02 [{shapeType=Capsule, length=4.096, radius=0.529, shapeName=dyn_shortDice_02_capsule, center=<2 0 0>, rotation=<rot 90 0 0>}]
 dyn_shortDice_03 [{shapeType=Capsule, length=0.4271, radius=2.5728, shapeName=dyn_shortDice_02_capsule, center=<1.863 0.5447 0.7625>, rotation=<rot 90 0 0>}]
 pelvis [{shapeType=Capsule, length=9.6815, radius=12.2728, shapeName=pelvis_capsule, center=<-1.8522 -1.386 0>, rotation=<rot 0 -90 0>}]
 spine_01 [{shapeType=Capsule, length=15.5573, radius=9.2327, shapeName=spine_01_capsule, center=<1.7545 0 0>, rotation=<rot 0 -90 0>}]
 thigh_l [{shapeType=Capsule, length=34.8554, radius=9.3541, shapeName=thigh_r_capsule, center=<-22.1907 -0.6066 2.52>, rotation=<rot 90 0 0>}]
 thigh_r [{shapeType=Capsule, length=34.6616, radius=9.5479, shapeName=thigh_r_capsule, center=<22.1907 0.6066 -2.52>, rotation=<rot 90 0 0>}]
}
constraints [
 {bone1Name=dyn_longDice_00, bone2Name=base_longDice_00, swing1LimitDegrees=80, swing2LimitDegrees=80, twistLimitDegrees=1}
 {bone1Name=dyn_longDice_01, bone2Name=dyn_longDice_00, swing1LimitDegrees=55, swing2LimitDegrees=55, twistLimitDegrees=1}
 {bone1Name=dyn_longDice_02, bone2Name=dyn_longDice_01, swing1LimitDegrees=45, swing2LimitDegrees=45, twistLimitDegrees=1}
 {bone1Name=dyn_longDice_03, bone2Name=dyn_longDice_02, swing1LimitDegrees=40, swing2LimitDegrees=40, twistLimitDegrees=1}
 {bone1Name=dyn_shortDice_00, bone2Name=base_shortDice_00, swing1LimitDegrees=80, swing2LimitDegrees=80, twistLimitDegrees=1}
 {bone1Name=dyn_shortDice_01, bone2Name=dyn_shortDice_00, swing1LimitDegrees=55, swing2LimitDegrees=55, twistLimitDegrees=1}
 {bone1Name=dyn_shortDice_02, bone2Name=dyn_shortDice_01, swing1LimitDegrees=45, swing2LimitDegrees=45, twistLimitDegrees=1}
 {
  bone1Name=base_longDice_00
  bone2Name=base_dice
  swing1LimitDegrees=45
  swing1Motion=Free
  swing2LimitDegrees=45
  swing2Motion=Free
  twistLimitDegrees=45
  twistMotion=Free
 }
 {
  bone1Name=base_shortDice_00
  bone2Name=base_dice
  swing1LimitDegrees=45
  swing1Motion=Free
  swing2LimitDegrees=45
  swing2Motion=Free
  twistLimitDegrees=45
  twistMotion=Free
 }
 {
  bone1Name=base_dice
  bone2Name=Melee_TwoHanded_Handle
  swing1LimitDegrees=45
  swing1Motion=Free
  swing2LimitDegrees=45
  swing2Motion=Free
  twistLimitDegrees=45
  twistMotion=Free
 }
 {bone1Name=dyn_longDice_04, bone2Name=dyn_longDice_03, swing1LimitDegrees=35, swing2LimitDegrees=35, twistLimitDegrees=1}
 {bone1Name=dyn_shortDice_03, bone2Name=dyn_shortDice_02, swing1LimitDegrees=40, swing2LimitDegrees=40, twistLimitDegrees=1}
 {
  bone1Name=spine_01
  bone2Name=pelvis
  swing1LimitDegrees=45
  swing1Motion=Free
  swing2LimitDegrees=45
  swing2Motion=Free
  twistLimitDegrees=45
  twistMotion=Free
 }
 {
  bone1Name=thigh_l
  bone2Name=pelvis
  swing1LimitDegrees=45
  swing1Motion=Free
  swing2LimitDegrees=45
  swing2Motion=Free
  twistLimitDegrees=45
  twistMotion=Free
 }
 {
  bone1Name=thigh_r
  bone2Name=pelvis
  swing1LimitDegrees=45
  swing1Motion=Free
  swing2LimitDegrees=45
  swing2Motion=Free
  twistLimitDegrees=45
  twistMotion=Free
 }
 {
  bone1Name=calf_r
  bone2Name=thigh_r
  swing1LimitDegrees=45
  swing1Motion=Free
  swing2LimitDegrees=45
  swing2Motion=Free
  twistLimitDegrees=45
  twistMotion=Free
 }
]
